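; Minor Spire gameplay actor: a damageable boss objective with a weakpoint, health widget and encounter music.
; Replicated spire state drives music, shield material, weakpoint and hiding, while incoming damage is blocked when shielded or scaled down off the weakpoint.
; The server tracks nearby enemy players through perception and on death spawns loot by playlist team size.
; Empty hooks and focus-based target facing support extension; client objective updates are stubbed out.
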(event EventBeginPlay
  ;[1 1 1] Intialization
  (|CallParentFunction)
  (CallFunction|HotfixEnableOverlapping)
  (CallFunction|OnSpireStateChanged (Variables|Default|GetSpireState)) ; Set up the Spire for Initial State
  (CallFunction|BeginPlayUI)
  (Utilities|SetHealthBasedonPlaylistSize)
  (Default|ServerInitNearbyActorPerception)
  (bind _asspires_minor_anim_blueprint (Utilities|Casting|CastToSpires_Minor_AnimBlueprint (Components|SkeletalMesh|GetAnimInstance (Variables|Default|GetSkeletalMeshComponentBudgeted)))
    (:then
      (Variables|Default|SetAs%20Spires%20Minor%20Anim%20Blueprint _asspires_minor_anim_blueprint))
    (:CastFailed
      _))
  (CallFunction|OnInitializationFinished))

(event EventEndPlay
  (params (EndPlayReason :type EEndPlayReason%20Enum :default "Destroyed"))
  ;[1 1 1] Deinitialization
  (|CallParentFunction EndPlayReason)
  (CallFunction|EndPlayUI))

(event SideEffects|EventBP_ApplySideEffect
  (params (ReactionSource :type Object%20Reference) (TargetObject :type Object%20Reference) (SideEffectTriggerTag :type Gameplay%20Tag%20Structure%20%28by%20ref%29) (SideEffectTriggerType :type EBlueprintableSideEffectTriggerType%20Enum%20%28by%20ref%29 :default "TriggeredAsTarget")))

(event Custom|OnSpireStateChanged
  (params (State :type EMinorSpireState%20Enum :default "NewEnumerator0"))
  ;[1 1 1] State Change
  (Development|PrintString (Utilities|String|Append "Minor Spire State Changed: " (Utilities|String|EnumtoString (Variables|Default|GetSpireState))))
  (Default|CallOnStateChanged State)
  (CallFunction|SetMusicState State)
  (switch Utilities|FlowControl|Switch|SwitchonEMinorSpireState State
    (:NewEnumerator1
      (CallFunction|SetVulnerable false))
    (:NewEnumerator2
      (CallFunction|SetVulnerable true))
    (:NewEnumerator3
      (Rendering|SetActorHiddenInGame true))))

(event Custom|Set%20Music%20State
  (params (Event%20Enum :type Byte))
  ;[1 1 1] State Change
  (MusicEncounter|ExecuteMusicEncounterGameplayEventByEnum (GameplayEventRouter|GetGlobalEventRouter) Event%20Enum self))

(event Custom|OnNearbyActorDetected
  (params (AllFoundActors :type Array%20of%20Actor%20Object%20References) (EnteringActors :type Array%20of%20Actor%20Object%20References) (ExitingActors :type Array%20of%20Actor%20Object%20References))
  ;[1 1 1] Detect Players
  (Networking|SwitchHasAuthority
    (:Authority
      (CallFunction|AddEnteringActorstoSensedList EnteringActors)
      (CallFunction|RemoveExitingActorsfromSensedList ExitingActors)
      (Variables|InternalData|SetSensedPlayerPawnCount (Utilities|Array|Length (Variables|InternalData|GetSensedPlayerPawns))))))

(event Custom|StartTargetFacing
  (params (CurrentTarget :type Fort%20Player%20Pawn%20Object%20Reference))
  ;[1 1 1] Face Player
  (bind _output_get (Variables|InternalData|SetCurrentTarget CurrentTarget)) ; Try to only set here to maintain sync with focus comp
  (Class|FortActorFocusComponent|SetActorFocus (Variables|Default|GetFortActorFocus) _output_get))

(event Custom|StopTargetFacing
  ;[1 1 1] Face Player
  (Class|FortActorFocusComponent|ClearFocus (Variables|Default|GetFortActorFocus)))

(event EventOnDeathServer
  (params (Damage :type Float%20%28single-precision%29) (DamageTags :type Gameplay%20Tag%20Container%20Structure%20%28by%20ref%29) (Momentum :type Vector :default "0, 0, 0") (HitInfo :type Hit%20Result%20Structure%20%28by%20ref%29) (InstigatedBy :type Controller%20Object%20Reference) (DamageCauser :type Actor%20Object%20Reference) (EffectContext :type Gameplay%20Effect%20Context%20Handle%20Structure))
  ;[1 1 1] Death - Loot Spawning and Weak point Destroy
  (CallFunction|SetSpireState "NewEnumerator3")
  (Utilities|FlowControl|Sequence
    (:then_0
      (bind _asbga_weakpoint_actor_minor_spire (Utilities|Casting|CastToBGA_WeakpointActor_MinorSpire (Variables|/Script/Engine.ChildActorComponent|GetChildActor (Variables|Default|GetWeakpoint01))))
      (Class|BGAWeakpointActorBase|DestroyWeakpoint _asbga_weakpoint_actor_minor_spire))
    (:then_1
      (Class|AthenaFunctionLibrary|SpawnLootfromTableUsingItemEntry (Utilities|Map|Find (Variables|Default|GetLoot%20Tier%20Group%20Names) (Variables|/Script/FortniteGame.FortPlaylist|GetMaxTeamSize (Variables|Default|GetPlaylist))) 1 -1 (Transformation|GetActorLocation) "0, 0, 0" 0 true true false 0 "Other" "NPCElimination")
      (GameplayCue|ExecuteGameplayCueOnActor(Burst) self (Variables|Default|GetGC_Destroyed) (CallFunction|GetEmptyParams)))))

(event EventOnDamagePlayEffects
  (params (Damage :type Float%20%28single-precision%29) (DamageTags :type Gameplay%20Tag%20Container%20Structure%20%28by%20ref%29) (Momentum :type Vector :default "0, 0, 0") (HitInfo :type Hit%20Result%20Structure%20%28by%20ref%29) (InstigatedBy :type Fort%20Pawn%20Object%20Reference) (DamageCauser :type Actor%20Object%20Reference) (EffectContext :type Gameplay%20Effect%20Context%20Handle%20Structure))
  ;[1 1 1] Play Hit Impact FX based on type of point that was hit
  (if (> Damage 0)
    (Class|SpiresMinorAnimBlueprint|SetHitReactBody (Variables|Default|GetAs%20Spires%20Minor%20Anim%20Blueprint))))

(event Custom|Activate%20Music
  ; Music activation
  (MusicEncounter|ExecuteMusicEncounterGameplayEventByName (GameplayEventRouter|GetGlobalEventRouter) "Activate" self))

(event Custom|Deactivate%20Music
  ; Music activation
  (MusicEncounter|ExecuteMusicEncounterGameplayEventByName (GameplayEventRouter|GetGlobalEventRouter) "Deactivate" self))

(event Custom|ForceStateServer
  (params (Index :type Integer))
  (Variables|Default|SetSpireState (Utilities|Enum|BytetoEnumEMinorSpireState (Math|Conversions|ToByte(Integer) Index))))

(event Custom|BeginPlay_UI
  ;[0.521 0.402 1 0.502] BeginPlay
  (Networking|SwitchHasAuthority
    (:Remote
      (CallFunction|OnBeginPlayCacheWidgetRef)
      (CallFunction|OnBeginPlayBindHealthWidgetEvents)
      (UI|UpdateObjectiveInfoServer false)
      (UI|UpdateObjectiveInfoClient))))

(event Custom|OnBeginPlay_CacheWidgetRef
  ;[0.521 0.402 1 0.502] BeginPlay
  ;;[0.521 0.402 1 0.502] OnBeginPlay - Cache Widget Ref
  (bind _healthstatuswidget (Variables|Default|GetHealthStatusWidget)
    (:then
      (bind _aswbp_minor_spire_objective_status (Utilities|Casting|CastToWBP_MinorSpire_ObjectiveStatus (UserInterface|GetWidget _healthstatuswidget)))
      (Variables|HealthStatusWidget|SetHealthAndObjectiveStatusWidget _aswbp_minor_spire_objective_status))
    (:else)))

(event Custom|OnBeginPlay_BindHealthWidgetEvents
  ;[0.521 0.402 1 0.502] BeginPlay
  ;;[0.521 0.402 1 0.502] OnBeginPlay - Bind to OnHealthChanged
  (Health|BindEventtoOnBuildingHealthChanged (EventDispatchers|CreateEvent :!event Custom|OnHealthChanged))
  (CallFunction|OnHealthChanged) ; Initialize on BeginPlay in case the spire is streamed in/out
)

(event Custom|EndPlay_UI
  ;[0.521 0.402 1 0.502] EndPlay
  (Networking|SwitchHasAuthority
    (:Remote
      (Health|UnbindEventfromOnBuildingHealthChanged (EventDispatchers|CreateEvent :!event Custom|OnHealthChanged)))))

(event Custom|OnHealthChanged
  ;[0.521 0.402 1 0.502] OnHealthChanged - Update the healthbar widget
  (bind _healthandobjectivestatuswidget (Variables|HealthStatusWidget|GetHealthAndObjectiveStatusWidget)
    (:then
      (Default|SetHealthPercent _healthandobjectivestatuswidget (Health|GetHealthPercent)))
    (:else)))

(fn AddEnteringActorsToSensedList
  (params (Array :type Array%20of%20Actor%20Object%20References))
  (for _array_element Array
    (if (Default|IsValidTarget _array_element)
      (Utilities|Array|AddUnique (Variables|InternalData|GetSensedPlayerPawns) (.AsFort%20Player%20Pawn (Default|IsValidTarget _array_element)))
      (CallFunction|OnPlayerSensed)
      (Development|PrintString (Utilities|String|Append "Added Player " (Utilities|GetDisplayName (.AsFort%20Player%20Pawn (Default|IsValidTarget _array_element))) " to sensed list."))))
  (return))

(fn BlueprintModifyIncomingDamage
  (params (Damage :type Float%20%28single-precision%29) (DamageTags :type Gameplay%20Tag%20Container%20Structure%20%28by%20ref%29) (EffectContext :type Gameplay%20Effect%20Context%20Handle%20Structure%20%28by%20ref%29) (EventInstigator :type Controller%20Object%20Reference) (DamageCauser :type Actor%20Object%20Reference))
  (|CallParentFunction Damage DamageTags EffectContext EventInstigator DamageCauser)
  (switch Utilities|FlowControl|Switch|SwitchonEMinorSpireState (Variables|Default|GetSpireState)
    (:NewEnumerator1
      (CallFunction|PlayFXForHitResult :Hit%20Type "NewEnumerator0")
      (return 0.0))
    (:NewEnumerator2
      (bind _bisweakpoint (Class|BPActorWeakpointsUtilityComponent|IsOwnedWeakpointActor (Variables|Default|GetBP_ActorWeakpoints_UtilityComponent) (Variables|GetDamageCauser) false))
      (CallFunction|PlayFXForHitResult (Ability|EffectContext|GetHitResult (Variables|GetEffectContext)) (select _bisweakpoint "NewEnumerator1" "NewEnumerator2"))
      (return (select _bisweakpoint (Variables|GetDamage) (* (Variables|GetDamage) (ScalableFloat|GetValueatLevel (Variables|Default|GetBaseDamagePercentage))))))))

(fn ForceSpireState
  (params (StateID :type Integer))
  (CallFunction|ForceStateServer StateID))

(fn GetEmptyParams
  (return (Ability|GameplayCue|MakeGameplayCueParameters 0.0)))

(fn HotfixEnableOverlapping
  ;[1 1 1] FORT-912400 - Disabled overlaps by default in details panel but HF enable them if gameplay issues arise post-pencils down
  (if (ScalableFloat|GetValueasBool (Variables|Default|GetHF_EnableUpdateOverlap))
    (Variables|/Script/Engine.PrimitiveComponent|SetbGenerateOverlapEvents true (Variables|Default|GetCapsule))
    (Collision|SetCollisionResponsetoChannel (Variables|Default|GetCapsule) "ECC_GameTraceChannel2" "ECR_Overlap")
    (Collision|SetCollisionResponsetoChannel (Variables|Default|GetCapsule) "ECC_GameTraceChannel14" "ECR_Overlap")
    (Collision|SetCollisionResponsetoChannel (Variables|Default|GetCapsule) "ECC_GameTraceChannel16" "ECR_Overlap")))

(fn isValidTarget
  (params (Object :type Object%20Reference))
  (bind _asfort_player_pawn (Utilities|Casting|CastToFortPlayerPawn Object))
  (return (and (not (GameplayTags|HasMatchingGameplayTag _asfort_player_pawn "(TagName=\"Pawn.Athena.NPC\")")) (not (Team|OnSameTeam _asfort_player_pawn self))) _asfort_player_pawn))

(fn OnInitializationFinished)

(fn OnPlayerSensed)

(fn OnRep_bIsObjectiveActive
  (UI|UpdateObjectiveInfoClient))

(fn OnRep_CurrentObjectiveCount
  (UI|UpdateObjectiveInfoClient))

(fn OnRep_MaxObjectiveCount
  (UI|UpdateObjectiveInfoClient))

(fn OnRep_SpireState
  (CallFunction|OnSpireStateChanged (Variables|Default|GetSpireState)))

(fn PlayFXForHitResult
  (params (Hit%20Result :type Hit%20Result%20Structure) (Hit%20Type :type EMinorSpireHitType%20Enum :default "NewEnumerator2"))
  (bind _hit_type (Variables|GetHit%20Type))
  (Development|PrintString (Utilities|String|Append (Utilities|String|EnumtoString _hit_type) " Hit Played"))
  (GameplayCue|ExecuteGameplayCueOnActor(Burst) self (Utilities|Select :Index _hit_type :NewEnumerator2 (Variables|Default|GetGC_HitImpact) :NewEnumerator1 (Variables|Default|GetGC_HitImpactWeakpoint) :NewEnumerator0 (Variables|Default|GetGC_Ricochet)) (GameplayCue|MakeGameplayCueParametersfromHitResult (Variables|GetHit%20Result))))

(fn RemoveExitingActorsFromSensedList
  (params (Array :type Array%20of%20Actor%20Object%20References))
  (for _array_element Array
    (bind _asfort_player_pawn (Utilities|Casting|CastToFortPlayerPawn _array_element))
    (if (Utilities|Array|ContainsItem (Variables|InternalData|GetSensedPlayerPawns) _asfort_player_pawn)
      (Utilities|Array|RemoveItem (Variables|InternalData|GetSensedPlayerPawns) _asfort_player_pawn)
      (Development|PrintString (Utilities|String|Append "Removed Player " (Utilities|GetDisplayName _asfort_player_pawn) " from sensed list."))))
  (return))

(fn Server_InitNearbyActorPerception
  (if (Networking|IsServer)
    (bind _returnvalue (Query|BPRegisterListener (Variables|Default|GetFortNearbyActorsPerception) (Utilities|Struct|MakeFortNearbyActorsPerceptionConfiguration 1.0 3 0 (ScalableFloat|GetValueatLevel (Variables|Default|GetPerceptionRadius)) "Friendly" false "(TokenStreamVersion=0,TagDictionary=((TagName=\"Athena.AI.IgnoreAll\"),(TagName=\"Gameplay.Status.DBNO\")),QueryTokenStream=(0,1,3,2,0,1),UserDescription=\"\",AutoDescription=\" NONE( Athena.AI.IgnoreAll, Gameplay.Status.DBNO )\")") (EventDispatchers|CreateEvent :!event onNearbyActorDetected))) ; Set Up Perception
    (Variables|Default|SetPerceptionListenerHandle _returnvalue)
    (return)))

(fn SetSpireState
  (params (NewSpireState :type EMinorSpireState%20Enum :default "NewEnumerator0"))
  (if (Utilities|Enum|NotEqual(Enum) NewSpireState (Variables|Default|GetSpireState))
    (Variables|Default|SetSpireState (Variables|GetNewSpireState)) ; Set active if not already
  ))

(fn SetVulnerable
  (params (bIsVultnerable :type Boolean))
  (Class|WBPMinorSpireObjectiveStatus|UpdateHealthInvulnerabilityState (Variables|HealthStatusWidget|GetHealthAndObjectiveStatusWidget) (Variables|GetbIsVultnerable))
  (Rendering|Material|SetOverlayMaterial (Variables|Default|GetSkeletalMeshComponentBudgeted) (select (Variables|GetbIsVultnerable) 0 "/SpireSharedAssets/Assets/Materials/M_MinorSpire_Shield.M_MinorSpire_Shield") false 0 :!class /Script/Engine.MeshComponent) ; Set the shield material override based on vulnerability
  (bind _asbga_weakpoint_actor_minor_spire (Utilities|Casting|CastToBGA_WeakpointActor_MinorSpire (Variables|/Script/Engine.ChildActorComponent|GetChildActor (Variables|Default|GetWeakpoint01))))
  (if (Variables|GetbIsVultnerable)
    (Class|BGAWeakpointActorBase|EnableWeakpoint _asbga_weakpoint_actor_minor_spire) ; Disabling a weakpoint is one way so we \ndon't have audio logic for enabling
    (else
      (Class|BGAWeakpointActorBase|DisableWeakpoint _asbga_weakpoint_actor_minor_spire)
      (Audio|Components|Audio|FadeOut (Variables|Default|GetMinorSpire_Shield_Loop) 0.5 0.0 "Linear" :!class /Script/Engine.AudioComponent))))

(fn UpdateObjectiveInfo_Client
  ;[0.15 0.007492 0.008924 0.5] Leaving this unwired for now to test without the Objuective UI
  (return))

(fn UpdateObjectiveInfo_Server
  (params (bIsObjectiveActive :type Boolean) (CurrentObjectiveCount :type Integer) (MaxObjectiveCount :type Integer))
  (Networking|SwitchHasAuthority
    (:Authority
      (Variables|Default|SetbIsObjectiveActive bIsObjectiveActive)
      (Variables|Default|SetCurrentObjectiveCount CurrentObjectiveCount)
      (Variables|Default|SetMaxObjectiveCount MaxObjectiveCount))
    (:Remote
      (return))))

(fn ConstructionScript)
