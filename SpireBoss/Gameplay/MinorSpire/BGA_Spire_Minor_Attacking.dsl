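; Attacking Minor Spire actor for the Spire boss, a laser-firing enemy driven by a StateTree.
; On authority BeginPlay it equips its ability set and starts the StateTree, which receives events for damage and sensed players.
; Laser charge events swap turning strategy and animate emissive brightness; reset also randomizes the attack delay.
; A fast, boosting vehicle can destroy it in a specific spire state if allowed, receiving knockback and triggering an impact cue.
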(event EventBeginPlay
  ; Init - Equip Ability Set and Start State Tree
  (|CallParentFunction)
  (Networking|SwitchHasAuthority
    (:Authority
      (Ability|EquipFortAbilitySet self (Variables|Default|GetAbility%20Set)) ; Inherent Ability Sets bugged when spawning from Scenario
      (Variables|Default|SetDelayBetweenAttacks (ScalableFloat|GetValueatLevel (Variables|Default|GetHF_MaximumDelayBetweenAttacks))) ; Setting up scalable float param in ST wait task bugs out param
      (AI|Logic|StartLogic (Variables|Default|GetStateTree)))))

(event EventOnDamageServer
  (params (Damage :type Float%20%28single-precision%29) (DamageTags :type Gameplay%20Tag%20Container%20Structure%20%28by%20ref%29) (Momentum :type Vector :default "0, 0, 0") (HitInfo :type Hit%20Result%20Structure%20%28by%20ref%29) (InstigatedBy :type Controller%20Object%20Reference) (DamageCauser :type Actor%20Object%20Reference) (EffectContext :type Gameplay%20Effect%20Context%20Handle%20Structure))
  ; Damage Received
  (Gameplay|StateTree|SendStateTreeEvent (Variables|Default|GetStateTree) (Utilities|Struct|MakeStateTreeEvent "(TagName=\"StateTree.Event.MinorSpire.DamageReceived\")") :!class /Script/GameplayStateTreeModule.StateTreeComponent))

(event OnComponentHit(Capsule)
  (params (HitComponent :type Primitive%20Component%20Object%20Reference) (OtherActor :type Actor%20Object%20Reference) (OtherComp :type Primitive%20Component%20Object%20Reference) (NormalImpulse :type Vector :default "0, 0, 0") (Hit :type Hit%20Result%20Structure%20%28by%20ref%29))
  ; Vehicle destroys Spire Logic
  (if (Utilities|DoesObjectImplementInterface OtherActor "/Script/FortniteGame.FortVehicleInterface")
    (Networking|SwitchHasAuthority
      (:Authority
        (bind _returnvalue (Vehicle|GetForwardSpeedKmh(Message) OtherActor))
        (bind _returnvalue_2 (Class|FortVehicleInterface|IsBoosting(Message) OtherActor))
        (if (Math|Boolean|ANDBoolean _returnvalue_2 (>= _returnvalue 2e+01) (Utilities|Enum|Equal(Enum) (Variables|Default|GetSpireState) "NewEnumerator2") (> (ScalableFloat|GetValueatLevel (Variables|Default|GetHF_AllowVehicleDestruction)) 0))
          (Gameplay|Die :DamageCauser OtherActor)
          (GameplayEffects|ApplyGameplayEffectToSelf :self (Ability|GetFortAbilitySystemComponent OtherActor) :GameplayEffectClass "/SpireSharedAssets/Abilities/MinorSpire/GE_MinorSpire_Vehicle_Knockback.GE_MinorSpire_Vehicle_Knockback_C" :Level 0.0)
          (GameplayCue|ExecuteGameplayCueOnActor(Burst) self (Variables|Default|GetGC_VehicleImpactDestroyed) (Ability|GameplayCue|MakeGameplayCueParameters :NormalizedMagnitude 0.0 :RawMagnitude 0.0 :Location (Math|Vector|MakeVector (.x (Transformation|GetActorLocation OtherActor)) (.y (Transformation|GetActorLocation OtherActor)) (+ (.z (Transformation|GetActorLocation OtherActor)) 4e+02)) :Normal "0,0,0" :GameplayEffectLevel 1 :AbilityLevel 1 :bReplicateLocationWhenUsingMinimalRepProxy false)))))))

(event Custom|StartLazerCharge
  ; Laser Charge
  (Class|FortActorFocusComponent|SetTurningStrategy (Variables|Default|GetFortActorFocus) (Variables|Default|GetTurningStrategyLazerSequence))
  (CallFunction|SetEmissiveClients true))

(event Custom|SetEmissiveClients
  (params (Increase :type Boolean))
  ; Laser Charge
  (if Increase
    (bind (_direction _newtrack_0) (Timeline|Timeline PlayFromStart)
      (:Update
        (Rendering|Material|SetScalarParameterValueonMaterials (Variables|Default|GetSkeletalMeshComponentBudgeted) "EmissiveBrightness" _newtrack_0))) ; Intended as a temp solution until SKM anims
    (else
      (Timeline|Timeline Reverse))))

(event Custom|ResetLazerCharge
  ; Laser Charge
  (Class|FortActorFocusComponent|SetTurningStrategy (Variables|Default|GetFortActorFocus) (Variables|Default|GetTurnStrategyDefault))
  (CallFunction|SetEmissiveClients false)
  (Variables|Default|SetDelayBetweenAttacks (Math|Random|RandomFloatinRange (ScalableFloat|GetValueatLevel (Variables|Default|GetHF_MinimumDelayBetweenAttacks)) (ScalableFloat|GetValueatLevel (Variables|Default|GetHF_MaximumDelayBetweenAttacks)))) ; Setting up scalable float param in ST wait task bugs out param
)

(fn OnPlayerSensed
  (|CallParentFunction)
  (Gameplay|StateTree|SendStateTreeEvent (Variables|Default|GetStateTree) (Utilities|Struct|MakeStateTreeEvent "(TagName=\"StateTree.Event.MinorSpire.PlayerDetected\")") :!class /Script/GameplayStateTreeModule.StateTreeComponent))

(fn ConstructionScript
  (|CallParentFunction))
